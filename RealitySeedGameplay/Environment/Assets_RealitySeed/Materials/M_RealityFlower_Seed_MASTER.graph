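# Master material for the Reality Seed flower, built on the shared character material function chain with optional spinning and bobbing motion.
# PreFX and CharacterSkinFX supply the attributes, which then pass through PostFX and cinematic PostFX functions.
# World Position Offset combines local Z rotation, blended between fixed angle and randomly varied time-driven spin, with sine-driven vertical bobbing.
# Normals are corrected for the rotation, and looping spin and bobbing default to off.
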Material /RealitySeedGameplay/Environment/Assets_RealitySeed/Materials/M_RealityFlower_Seed_MASTER
graph {
 properties {bUseMaterialAttributes=true, pixelDepthOffsetMode=PDOM_Legacy, refractionMethod=RM_IndexOfRefraction, usedShadingModels=Default Lit}
 nodes [
  n1 = MaterialFunctionCall() {materialFunction=&BaseFN_CharacterFunction_PreFX}
  n2 = MaterialFunctionCall(
   PreskinnedPosition=@n1.Pre skinned Local Position
   Gmap=@n1.Pre skinned Local Position
   MaterialAttributes=@n1.Pre skinned Local Position
   AO=@n1.Pre skinned Local Position
  ) {materialFunction=&CharacterSkinFX}
  n3 = Constant3Vector() {constant=<rgba 0 0 1 0>}
  n4 = ScalarParameter() {defaultValue=1, parameterName=SpinningSpeed}
  n5 = Multiply(A=ScalarParameter() {defaultValue=0.5, parameterName=MotionRandomInput}) {constB=2}
  n6 = Multiply(A=@n4, B=Multiply(A=Subtract(A=@n5)) {constB=0.3})
  n7 = LinearInterpolate(
   A=ScalarParameter() {parameterName=RotationAngle}
   B=Multiply(A=Time(), B=Add(A=@n4, B=@n6))
   Alpha=ScalarParameter() {parameterName=Looping Rotation}
  )
  n8 = Transform(None=RotateAboutAxis(
    NormalizedRotationAxis=@n3
    RotationAngle=@n7
    PivotPoint=Constant()
    Position=MaterialFunctionCall() {materialFunction=&LocalPosition}.Local Position
   )) {transformSourceType=TRANSFORMSOURCE_Local}
  n9 = Constant()
  n10 = MaterialFunctionCall(Speed=ScalarParameter() {defaultValue=0.5, parameterName=BobbingSpeed}) {materialFunction=&DebugTimeSine}
  n11 = ScalarParameter() {defaultValue=10, parameterName=BobbingVerticalHeight}
  n12 = MaterialFunctionCall(X=@n9, Y=@n9, Z=LinearInterpolate(B=Multiply(A=@n10.Result, B=@n11) {constB=10}, Alpha=ScalarParameter() {parameterName=BobbingAmount})) {materialFunction=&MakeFloat3}
  n13 = Transform(None=GetMaterialAttributes(None=@n2.Result) {attributeGetTypes [#1]}.Normal)
  n14 = MaterialFunctionCall(
   AO=@n1.AO
   Distance Blend=@n1.AO
   Fresnel=@n1.AO
   Gmap=@n1.AO
   Pre Skinned Local Position=@n1.AO
   Material=SetMaterialAttributes!(
    MaterialAttributes=@n2.Result
    World Position Offset=Add(A=@n8, B=@n12.Result)
    Normal=MaterialFunctionCall(Rotation Axis=@n3, Rotation Angle=@n7, World Space Vertex Normals=@n13) {materialFunction=&FixRotateAboutAxisNormals}.New Tangent Space Vertex Normal
   ) {attributeSetTypes [#2, #1]}
  ) {materialFunction=&BaseFN_CharacterFunction_PostFX}
  MaterialFunctionCall(In=@n14.Result) {materialFunction=&CINE_MF_CharPostFX}
 ]
}
